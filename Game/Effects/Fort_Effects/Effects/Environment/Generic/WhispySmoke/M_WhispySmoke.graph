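# Unlit translucent material for large environmental smoke plumes, apparently volcanic, built on SplineThicken geometry that widens from base to tip.
# Two panned, rotated smoke textures drive opacity, shaped by UV gradients, depth and near/far distance fades.
# Color blends bottom-to-top tints with time-of-day values and Schlick-phase sun scattering.
# A terrain heightmap check fades opacity and collapses offsets, while lower quality paths drop depth fade, hue shift and animated wobble.
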Material /Game/Effects/Fort_Effects/Effects/Environment/Generic/WhispySmoke/M_WhispySmoke
graph {
 properties {bUsedWithNanite=true, blendMode=BLEND_Translucent, pixelDepthOffsetMode=PDOM_Legacy, shadingModel=MSM_Unlit, usedShadingModels=Unlit}
 nodes [
  Constant() {r=0.1}
  Constant() {r=1}
  MaterialFunctionCall!(Axis, Exponent, Normal) {materialFunction=/Game/Effects/Fort_Effects/Materials/_Functions/MF_AxisAlignedFrensnel}
  MaterialFunctionCall() {materialFunction=&AxisAlignedFresnel}
  n1 = AppendVector(
   A=Constant() {r=1}
   B=LinearInterpolate(
    A=ScalarParameter() {defaultValue=0.5, group=Texture, parameterName=TilingSmoke_V_Mult_Min}
    B=ScalarParameter() {defaultValue=0.5, group=Texture, parameterName=TilingSmoke_V_Mult_Max, sortPriority=33}
    Alpha=VertexColor().R
   ) {constA=1.25, constB=2}
  )
  n2 = TextureCoordinate() {vTiling=0.35}
  n3 = Multiply(A=ComponentMask(None=@n2) {g=true}, B=LinearInterpolate(Alpha=VertexColor().R) {constA=1.25, constB=2})
  n4 = Frac(None=ComponentMask(None=ObjectPositionWS()) {r=true})
  n5 = Add(A=AppendVector(A=ComponentMask(None=@n2) {r=true}, B=Add(A=@n3, B=@n4)), B=AppendVector(A=Constant(), B=VertexColor().R))
  n6 = VectorParameter() {group=Texture, parameterName=PannerSpeed, defaultValue=<rgba 0 0.035 0 1>}
  n7 = MaterialFunctionCall(
   UVs=Panner(Coordinate=Multiply(A=@n1, B=@n5), Speed=ComponentMask(None=@n6.RGB) {g=true, r=true})
   "Rotation Angle (0-1)"=ScalarParameter() {defaultValue=0.25, group=Texture, parameterName=Rotation Angle Tex 1}
  ) {materialFunction=&CustomRotator}
  n8 = Multiply(
   A=AppendVector(A=ScalarParameter() {defaultValue=1.5, parameterName=DetailTileX}, B=ScalarParameter() {defaultValue=0.75, parameterName=DetailTileY})
   B=@n7.Rotated Values
  )
  n9 = TextureSampleParameter2D(UVs=@n8) {
   parameterName=DetalTexture_Tiling
   samplerType=SAMPLERTYPE_Normal
   texture=/Game/Global/Textures/Default/Tileable_Patterns/T_StormCloudDetail01_N
  }
  n10 = TextureSampleParameter2D(UVs=@n8) {
   parameterName=DetalTexture_Smoke
   samplerType=SAMPLERTYPE_Normal
   texture=/Game/Effects/Fort_Effects/Gameplay/SmokeDustExplosions/Textures/T_RibbonSmokeMasked_01
  }
  n11 = MaterialFunctionCall!(WorldSpaceNormal=Transform(None=Add(A=@n9.RGB, B=@n10.RGB)), MaskForLowEnd) {materialFunction=/Game/Athena/Prototype/Blueprints/White/FX/MF_Volcano_Smoke_Colors}
  n12 = Constant() {r=1}
  n13 = VectorParameter() {group=Color, parameterName=Color_Top, defaultValue=<rgba 2 2.75 5 1>}
  n14 = LinearInterpolate(A=ScalarParameter() {parameterName=HueShiftLow}, B=ScalarParameter() {defaultValue=0.15, parameterName=HueShiftHigh}, Alpha=VertexColor().R)
  n15 = ComponentMask(None=TextureCoordinate()) {r=true}
  n16 = ComponentMask(None=TextureCoordinate()) {g=true}
  n17 = Multiply(
   A=LinearInterpolate(
    A=VectorParameter() {group=Color, parameterName=Color_Bottom_Base, defaultValue=<rgba 10 1.3943 0 0>}.RGB
    B=VectorParameter() {group=Color, parameterName=Color_Bottom, defaultValue=<rgba 3 0.4183 0 0>}.RGB
    Alpha=Multiply(A=@n15, B=@n16)
   )
   B=MaterialFunctionCall(
    Value_Morning=ScalarParameter() {defaultValue=1, group=Color, parameterName="Value_Morning (S)"}
    Value_Day=ScalarParameter() {defaultValue=2, group=Color, parameterName="Value_Day (S)"}
    Value_Evening=ScalarParameter() {defaultValue=1, group=Color, parameterName="Value_Evening (S)"}
    Value_Night=ScalarParameter() {defaultValue=1.5, group=Color, parameterName="Value_Night (S)"}
   ) {materialFunction=&Scalar_Time_of_Day}.Result
  )
  n18 = QualitySwitch(
   Default=StaticSwitchParameter(
    True=MaterialFunctionCall(Hue Shift Percentage=@n14, Texture=@n17) {materialFunction=&HueShift}.Result
    False=@n17
   ) {defaultValue=true, parameterName=UseHueShift}
   Low=@n17
  )
  n19 = ShadingPathSwitch(Default=FeatureLevelSwitch(Default=@n18, ES2=@n17, ES3_1=@n17, SM4_REMOVED=@n18), Forward=@n17)
  n20 = Add(A=ComponentMask(None=CameraVectorWS()) {b=true}, B=ScalarParameter() {defaultValue=0.1, parameterName=offset})
  n21 = ScalarParameter() {defaultValue=0.2, parameterName=Div}
  n22 = TextureCoordinate()
  n23 = ComponentMask(None=VectorParameter() {group=Color, parameterName=TileColorMask, defaultValue=<rgba 1 0.35 5 1>}.RGB) {g=true, r=true}
  n24 = StaticSwitchParameter(True=Clamp(None=Divide(A=@n20, B=@n21)), False=ComponentMask(None=Multiply(A=@n22, B=@n23)) {g=true}) {defaultValue=true, parameterName=UseHeightInsteadofUVtofade}
  n25 = CollectionParameter() {
   collection=&FortniteMaterialParameters
   parameterName=SunlightDirectionalVector
  }
  MaterialExpressionDotProduct_1 = DotProduct(A=ComponentMask(None=@n25) {b=true, g=true, r=true}, B=CameraVectorWS())
  MaterialExpressionScalarParameter_5 = ScalarParameter() {defaultValue=0.4, parameterName=Anisotropy}
  n26 = Custom!(CosTheta=@MaterialExpressionDotProduct_1, k=@MaterialExpressionScalarParameter_5) {
   code="\tfloat Inner = (1 + k * CosTheta);\r\n\treturn (1 - k * k) / (4 * PI * Inner * Inner);\r\n\r\n"
   description=Schlick Phase
   inputs [
    {
     inputName=CosTheta
     input {expression=&self:MaterialExpressionDotProduct_1}
    }
    {
     inputName=k
     input {expression=&self:MaterialExpressionScalarParameter_5}
    }
   ]
  }
  n27 = MaterialFunctionCall(
   Color_Morning=VectorParameter() {group=Color, parameterName="Color_Morning (V3)", defaultValue=<rgba 1.25 0.9257 0.708 0>}.RGB
   Color_Day=VectorParameter() {group=Color, parameterName="Color_Day (V3)", defaultValue=<rgba 1.25 0.5722 0.5598 0>}.RGB
   Color_Evening=VectorParameter() {group=Color, parameterName="Color_Evening (V3)", defaultValue=<rgba 1 0.6005 0.2592 0>}.RGB
   Color_Night=VectorParameter() {group=Color, parameterName="Color_Night (V3)", defaultValue=<rgba 1 0.4906 0.3 0>}.RGB
  ) {materialFunction=&Color_Time_of_Day}
  n28 = Multiply(A=Multiply(A=@n26, B=ScalarParameter() {defaultValue=25, parameterName=ScaleCloudColor}), B=Multiply(A=@n27.Result))
  n29 = VectorParameter() {group=Color, parameterName=Color_Whole, defaultValue=<rgba 2 2.75 5 1>}
  n30 = QualitySwitch(
   Default=MaterialFunctionCall(Hue Shift Percentage=@n14, Texture=@n29.RGB) {materialFunction=&HueShift}.Result
   Low=@n29.RGB
  )
  n31 = StaticSwitchParameter(
   True=Multiply(A=LinearInterpolate(A=@n13.RGB, B=@n19, Alpha=@n24), B=Multiply(A=@n28, B=@n27.Result))
   False=ShadingPathSwitch(Default=FeatureLevelSwitch(Default=@n30, ES2=@n29.RGB, ES3_1=@n29.RGB, SM4_REMOVED=@n30), Forward=@n29.RGB)
  ) {defaultValue=true, parameterName=UseTOD?}
  NamedRerouteDeclaration(None=Multiply(A=StaticSwitchParameter(True=@n11.Result, False=@n12) {parameterName=UseDetailTexture?}, B=Multiply(A=@n31, B=@n16))) {name=EMISSIVE_COLOR, nodeColor=<rgba 1 0.2344 0 1>}
  n32 = ScalarParameter() {defaultValue=5000, parameterName=DepthFadeDistance}
  n33 = TextureSampleParameter2D(UVs=@n7.Rotated Values) {parameterName=SmokeTexture_01, texture=/Game/Effects/Fort_Effects/Textures/Smoke/T_Puffy_Smoke_Trail_01}
  n34 = ScalarParameter() {defaultValue=0.5, group=Texture, parameterName=TilingSmoke_U_Mult_Tex2, sortPriority=34}
  n35 = LinearInterpolate(
   A=ScalarParameter() {defaultValue=0.5, group=Texture, parameterName=TilingSmoke_V_Mult_Tex2_Min, sortPriority=34}
   B=ScalarParameter() {defaultValue=0.5, group=Texture, parameterName=TilingSmoke_V_Mult_Tex2_Max, sortPriority=35}
   Alpha=VertexColor().R
  ) {constA=1.25, constB=2}
  n36 = Panner(
   Coordinate=Multiply(A=@n5, B=AppendVector(A=@n34, B=@n35))
   Speed=ComponentMask(None=VectorParameter() {group=Texture, parameterName=PannerSpeed_02, defaultValue=<rgba 0 0.01 0 1>}.RGB) {g=true, r=true}
  )
  n37 = ScalarParameter() {defaultValue=0.25, group=Texture, parameterName=Rotation Angle Tex 2}
  n38 = Multiply(
   A=TextureSampleParameter2D(UVs=MaterialFunctionCall(UVs=@n36, "Rotation Angle (0-1)"=@n37) {materialFunction=&CustomRotator}.Rotated Values) {parameterName=SmokeTexture_02, texture=&T_Puffy_Smoke_Trail_01}.G
   B=ScalarParameter() {defaultValue=1, group=Texture, parameterName=SmokeTexture_02_Gain}
  )
  n39 = Multiply(A=Add(A=@n33.G, B=Multiply(A=@n33.G, B=@n38)), B=ScalarParameter() {defaultValue=0.25, group=Texture, parameterName=Opacity})
  n40 = ShadingPathSwitch(
   Default=ScalarParameter() {defaultValue=0.65, group=CloseFarFader, parameterName=DistanceScaleMult_LowQuality}
   Deferred=QualitySwitch(Default=Constant() {r=1}, Low=ScalarParameter() {defaultValue=0.65, group=CloseFarFader, parameterName=DistanceScaleMult_LowQuality})
  ) {desc=Squish WPO earlier on lower quality}
  n41 = Multiply(A=ScalarParameter() {defaultValue=25000, group=CloseFarFader, parameterName="FarFadeLength (S)"}, B=@n40)
  n42 = Multiply(A=ScalarParameter() {defaultValue=125000, group=CloseFarFader, parameterName="FarFadeOffset (S)"}, B=@n40)
  n43 = MaterialFunctionCall!(
   NearFadeLength=Divide(A=ScalarParameter() {defaultValue=21000, group=CloseFarFader, parameterName="NearFadeLength (S)"}, B=@n40)
   NearFadeOffset=Divide(A=ScalarParameter() {defaultValue=20000, group=CloseFarFader, parameterName="NearFadeOffset (S)"}, B=@n40)
   NearFadeCurve
   NearFadeStartingScale
   NearFadeEndingScale
   FarFadeLength=@n41
   FarFadeOffset=@n42
   FarFadeCurve
   FarFadeEndingScale
   Opacity Near Distance Bias
   Opacity Far Distance Bias=ScalarParameter() {defaultValue=-2, group=CloseFarFader, parameterName="Opacity Far Distance Bias (S)"}
  ) {materialFunction=/Game/Effects/Fort_Effects/Materials/_Functions/MF_SmoothLod}
  n44 = TextureCoordinate()
  n45 = ComponentMask(None=@n44) {g=true}
  n46 = ComponentMask(None=@n44) {g=true}
  n47 = Multiply(A=Multiply(A=@n39, B=@n43.Opacity), B=Multiply(A=Multiply(A=@n45, B=@n45), B=OneMinus(None=@n46)))
  n48 = QualitySwitch(Default=Multiply(A=DepthFade(FadeDistance=@n32), B=@n47), Low=@n47)
  n49 = ShadingPathSwitch(Default=@n48, Deferred=@n48, Forward=@n47)
  n50 = TextureSampleParameter2D(UVs=TextureCoordinate() {coordinateIndex=1}) {
   group=LavaHeightCheck
   parameterName=TerrainHeightMap
   samplerType=SAMPLERTYPE_LinearGrayscale
   texture=/Game/Environments/Sets/Vulcano/Textures/T_TerrainHeightMap
  }
  n51 = Subtract(
   A=LinearInterpolate(
    A=ScalarParameter() {defaultValue=-2500, group=LavaHeightCheck, parameterName=TerrainHeight Mininum}
    B=ScalarParameter() {defaultValue=11082.7412, group=LavaHeightCheck, parameterName=TerrainHeightMax}
    Alpha=@n50.R
   )
   B=ComponentMask(None=ActorPositionWS()) {b=true}
  )
  n52 = ScalarParameter() {defaultValue=1500, group=LavaHeightCheck, parameterName=TerrainHeightCheckRange}
  n53 = Clamp(None=OneMinus(None=Divide(A=@n51, B=@n52)))
  n54 = StaticSwitchParameter(True=Multiply(A=@n49, B=VertexInterpolator(VS=@n53).PS), False=@n49) {defaultValue=true, parameterName=UseHeightMapTexture?}
  n55 = ScalarParameter() {defaultValue=1500, group=WPO, parameterName=WidthBase_Min}
  n56 = ScalarParameter() {defaultValue=2500, group=WPO, parameterName=WidthBase_Max}
  n57 = VertexColor()
  n58 = MaterialFunctionCall(
   NearFadeLength=ScalarParameter() {defaultValue=500, group=CloseFarFader, parameterName=NearShrinkLength}
   NearFadeOffset=ScalarParameter() {defaultValue=500, group=CloseFarFader, parameterName=NearShrinkOffset}
   FarFadeLength=@n41
   FarFadeOffset=@n42
  ) {materialFunction=&MF_SmoothLod}
  n59 = ScalarParameter() {defaultValue=18000, group=WPO, parameterName=WidthTip_Min}
  n60 = ScalarParameter() {defaultValue=35000, group=WPO, parameterName=WidthTip_Max}
  n61 = VertexColor()
  n62 = MaterialFunctionCall(
   WidthBase=Multiply(A=LinearInterpolate(A=@n55, B=@n56, Alpha=@n57.R), B=@n58.Opacity)
   WidthTip=Multiply(A=LinearInterpolate(A=@n59, B=@n60, Alpha=@n61.R), B=@n58.Opacity)
  ) {materialFunction=&SplineThicken}
  n63 = DotProduct(A=CameraVectorWS(), B=@n62.Normal)
  n64 = Power(Base=Saturate(None=Abs(None=@n63)), Exp=ScalarParameter() {defaultValue=5, parameterName=FresnelExponent})
  NamedRerouteDeclaration(None=StaticSwitchParameter(True=Multiply(A=@n54, B=@n64), False=@n54) {parameterName=UseFresnel?}) {name=OPACITY, nodeColor=<rgba 0 0.0859 1 1>}
  n65 = Constant3Vector()
  n66 = VertexNormalWS()
  n67 = Constant3Vector() {desc=Disable WPO anim in Low, constant=<rgba 0 0 20 0>}
  n68 = VectorParameter() {group=WPO, parameterName=TextureWPO_UVtiling, defaultValue=<rgba 1 2 0 0>}
  n69 = Panner(Coordinate=Multiply(A=TextureCoordinate(), B=ComponentMask(None=@n68.RGB) {g=true, r=true})) {speedY=0.15}
  n70 = TextureSample(UVs=@n69) {samplerType=SAMPLERTYPE_Normal, texture=&T_Weap_Distort_N_02}
  n71 = AppendVector(
   A=Multiply(A=ComponentMask(None=@n70.RGB) {r=true}, B=ScalarParameter() {defaultValue=25, group=WPO, parameterName=WPO_HeightX}) {constB=0.5}
   B=Multiply(A=ComponentMask(None=@n70.RGB) {g=true}, B=ScalarParameter() {group=WPO, parameterName=WPO_HeightY}) {constB=0.5}
  )
  n72 = Multiply(
   A=TextureSample(UVs=@n69) {texture=&T_Smooth_tile}.G
   B=ScalarParameter() {group=WPO, parameterName=WPO_HeightZ}
  ) {constB=0.6}
  n73 = MaterialFunctionCall() {materialFunction=&LinearGradient}
  n74 = Constant() {r=0.5}
  n75 = Transform(None=LinearInterpolate(A=Constant3Vector(), B=AppendVector(A=@n71, B=@n72), Alpha=SphereMask(A=@n73.VGradient, B=@n74, Radius=@n74) {hardnessPercent=0}))
  n76 = LinearInterpolate(
   A=ScalarParameter() {group=WPO, parameterName=WPO_Wave_Intensity_Close}
   B=ScalarParameter() {defaultValue=100, group=WPO, parameterName=WPO_Wave_Intenstiy}
   Alpha=Clamp(None=@n43.Opacity)
  )
  n77 = ShadingPathSwitch(Default=@n67, Deferred=QualitySwitch(Default=Multiply(A=@n75, B=@n76) {constB=80}, Low=@n67))
  n78 = ShadingPathSwitch(Default=@n62.WorldPositionOffset, Deferred=QualitySwitch(Default=Add(A=@n62.WorldPositionOffset, B=@n77), Low=@n62.WorldPositionOffset))
  n79 = StaticSwitchParameter(True=LinearInterpolate(A=Multiply(A=@n65, B=@n66), B=@n78, Alpha=@n53), False=@n78) {defaultValue=true, parameterName=UseHeightMapTexture?}
  n80 = TextureCoordinate()
  n81 = RotateAboutAxis(
   NormalizedRotationAxis=ObjectOrientation()
   RotationAngle=Multiply(A=ComponentMask(None=@n80) {g=true}, B=ScalarParameter() {parameterName=RotateAngle})
   PivotPoint=MaterialFunctionCall() {materialFunction=&ObjectPivotPoint}.Object Pivot Location
   Position=WorldPosition().XYZ
  )
  NamedRerouteDeclaration(None=StaticSwitchParameter(True=Add(A=@n79, B=@n81), False=@n79) {parameterName=UseRotateBend?}) {name=WPO, nodeColor=<rgba 0.4062 1 0 1>}
  n82 = Multiply(A=@n33.G, B=ScalarParameter() {defaultValue=0.35, group=Texture, parameterName=OpacityLowEnd})
  ShadingPathSwitch(Default=QualitySwitch(Default=@n39, Low=@n82), Forward=@n82, Mobile=@n82)
 ]
}
